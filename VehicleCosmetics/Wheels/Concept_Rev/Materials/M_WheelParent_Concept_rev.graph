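# Parent material for the Concept_Rev vehicle cosmetic wheel, adding an animated, tintable tire emissive on top of MF_WheelParent_Simple.
# Panning and rotated noise samples build the glow and a warp signal, alongside a tire mask, object-relative height mask and color parameters.
# Vertex color red blends the base emissive toward a named-reroute value.
# The resulting attributes pass through the shared vehicle damage, then de-rez transition functions.
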Material /VehicleCosmetics/Wheels/Concept_Rev/Materials/M_WheelParent_Concept_rev
graph {
 properties {
  bAutomaticallySetUsageInEditor=false
  bUseMaterialAttributes=true
  bUsedWithMeshDeformer=true
  bUsedWithSkeletalMesh=true
  pixelDepthOffsetMode=PDOM_Legacy
  previewMesh=/VehicleCosmetics/Wheels/Concept_Rev/Meshes/SK_Wheel_ConceptRev
  usedShadingModels=Default Lit
 }
 nodes [
  n1 = MaterialFunctionCall() {materialFunction=&MF_WheelParent_Simple}
  n2 = GetMaterialAttributes(None=@n1.MaterialAttributes) {attributeGetTypes [#1]}
  n3 = LinearInterpolate(A=@n2.Emissive Color, B=NamedRerouteUsage(), Alpha=VertexColor().R)
  MaterialFunctionCall(In=MaterialFunctionCall(Material=SetMaterialAttributes(MaterialAttributes=@n2.MaterialAttributes, Emissive Color=@n3) {attributeSetTypes [#1]}) {materialFunction=&MF_BaseVehicle_Damage}.Result) {materialFunction=&MF_VehicleDeRez_Transition}
  n4 = NamedRerouteUsage()
  n5 = TextureCoordinate()
  n6 = Constant() {r=1}
  n7 = TextureObject() {texture=/VehicleCosmetics/Wheels/_SharedTextures/FX/DistortNoise_02_RGB_T}
  n8 = TextureSample(UVs=Rotator(Coordinate=Panner(Coordinate=@n5) {speedY=0.5}, Time=@n6) {speed=1.57}, Tex=@n7) {samplerSource=SSM_Wrap_WorldGroupSettings, texture=&DistortNoise_02_RGB_T}
  n9 = Panner(Coordinate=TextureCoordinate()) {speedX=0.3}
  n10 = TextureObject() {texture=/VehicleCosmetics/Wheels/_SharedTextures/FX/Universal_Noise_55_RGB}
  n11 = TextureSample(UVs=Panner(Coordinate=TextureCoordinate()) {speedX=0.15}, Tex=@n10) {samplerSource=SSM_Wrap_WorldGroupSettings, texture=&DistortNoise_02_RGB_T}
  n12 = Multiply(
   A=Multiply(
    A=@n8.R
    B=TextureSample(UVs=@n9, Tex=@n10) {samplerSource=SSM_Wrap_WorldGroupSettings, texture=&DistortNoise_02_RGB_T}.R
   )
   B=@n11.G
  )
  n13 = Add(A=Multiply(A=OneMinus(None=@n4), B=NamedRerouteUsage()), B=Multiply(A=@n12) {constB=20})
  n14 = VectorParameter() {group=5_Special Effects, parameterName=TIRE_Custom Color, sortPriority=0, defaultValue=<rgba 0 0.4735 0.3613 1>}
  n15 = Add(
   A=Multiply(A=NamedRerouteUsage(), B=ScalarParameter() {defaultValue=-0.15, group=5_Special Effects, parameterName=TIRE_Hue Shift, sortPriority=6})
   B=@n14.RGB
  )
  n16 = NamedRerouteUsage()
  n17 = Clamp(None=Add(A=LinearInterpolate(A=@n15, B=@n16, Alpha=@n16), B=@n16)) {maxDefault=2.5, minDefault=-50}
  NamedRerouteDeclaration(None=Multiply(
    A=Multiply(A=@n13, B=@n17)
    B=VectorParameter() {group=5_Special Effects, parameterName=PaintedColor, defaultValue=<rgba 0 0.1094 0.1084 1>}.RGB
   )) {name=Tire_Emissive, nodeColor=<rgba 1 1 0 1>}
  NamedRerouteDeclaration(None=ComponentMask(None=@n1.TireRGBAMask) {g=true}) {name=Tire_Mask_G, nodeColor=<rgba 0 1 0 1>}
  n18 = Multiply(A=@n8.B)
  n19 = Rotator(Coordinate=Panner(Coordinate=TextureCoordinate()) {speedX=0.25}, Time=@n6) {speed=1.57}
  n20 = Multiply(A=TextureSample(UVs=@n19, Tex=@n7) {samplerSource=SSM_Wrap_WorldGroupSettings, texture=&DistortNoise_02_RGB_T}.R) {constB=2}
  n21 = Panner(Coordinate=TextureCoordinate()) {speedX=0.4}
  n22 = Add(
   A=Multiply(A=TextureSample(UVs=@n21) {samplerSource=SSM_Wrap_WorldGroupSettings, texture=/VehicleCosmetics/Wheels/_SharedTextures/FX/Universal_Noise_58_RGB}.G)
   B=Multiply(
    A=@n11.G
    B=TextureSample(UVs=@n19, Tex=@n10) {samplerSource=SSM_Wrap_WorldGroupSettings, texture=&DistortNoise_02_RGB_T}.G
   )
  )
  NamedRerouteDeclaration(None=Add(A=Add(A=@n18, B=@n20), B=Add(A=@n22, B=@n12))) {name=Warp, nodeColor=<rgba 1 0 1 1>}
  n23 = Subtract(A=WorldPosition().XYZ, B=ObjectPositionWS())
  NamedRerouteDeclaration(None=ConstantBiasScale(None=ComponentMask(None=@n23) {b=true}) {bias=-16, scale=0.07}) {name=Z Mask, nodeColor=<rgba 0 0 1 1>}
  NamedRerouteUsage()
 ]
}
